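# LemoWrath skydiving contrail Niagara system, drawing three clusters of eye and trail meshes around the player, plus air, limb trails and sparks.
# Burst-spawned mesh particles take per-partition transforms from arrays, follow the skeletal mesh and bob via waveform.
# User.ParaGlideLeanAlpha crossfades air against eyes and trails, enlarges and oranges the four limb trails, and drives spark spawning.
# FrontEnd modifiers give menus separate placement, scale and opacity.
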NiagaraSystem /BRCosmetics/Athena/Effects/Particles/Skydiving/NS_Contrail_LemoWrath/NS_Contrail_LemoWrath02
dependencies {
 usedDataInterfaces [
  &NiagaraDataInterfaceArrayFloat3
  &NiagaraDataInterfacePlatformSet
  &NiagaraDataInterfaceArrayFloat
  &NiagaraDataInterfaceArrayInt32
  &NiagaraDataInterfaceCurve
  &NiagaraDataInterfaceSkeletalMesh
 ]
 usedDynamicInputs []
 usedModules [
  &self:SystemSpawnScript.SystemScriptSource.SystemScriptGraph.NiagaraNodeAssignment_1.SetVariables_#1
  &self:SystemSpawnScript.SystemScriptSource.SystemScriptGraph.NiagaraNodeAssignment_2.SetVariables_#2
  &self:SystemSpawnScript.SystemScriptSource.SystemScriptGraph.NiagaraNodeAssignment_3.SetVariables_#3
  &self:SystemSpawnScript.SystemScriptSource.SystemScriptGraph.NiagaraNodeAssignment_0.SetVariables_#4
  &self:SystemSpawnScript.SystemScriptSource.SystemScriptGraph.NiagaraNodeAssignment_4.SetVariables_#5
  &SystemState
 ]
}
emitters {
 Air {
  emitterUpdateScript [
   &EmitterState(
    Inactive Response=::NewEnumerator1
    Life Cycle Mode=::NewEnumerator1
    Loop Behavior=::NewEnumerator1
    Loop Duration Mode=::NewEnumerator1
    UseLoopDelay=false
   )
   &SpawnBurst_Instantaneous(Spawn Count=3, Use Spawn Probability=false, UseLoopCountLimit=false)
  ]
  particleSpawnScript [
   &PartitionParticles(Number Of Partitions=3)
   &InitializeParticle(
    Color=<rgba 1 1 1 0.1>
    Mesh Scale=&SelectVectorFromArray(
     Direct Array Index=$Particles.PartitionParticles.Partition
     Vector Selection Array=DataInterface() {propertyValues {FloatData [<0.6 0.6 0.6>, <0.6 0.6 0.6>, <0.6 0.6 0.6>]}}
    )
    Mesh Scale Mode=::NewEnumerator4
    Position Offset=&FN_FrontEndVector3Modifier(
     Gameplay=&AddVectorToPosition(
      Position=&SelectVectorFromArray(Direct Array Index=$Particles.PartitionParticles.Partition, Vector Selection Array=$System.Main Position)
      Vector=&SelectVectorFromArray(
       Direct Array Index=$Particles.PartitionParticles.Partition
       Vector Selection Array=DataInterface() {propertyValues {FloatData [<12 55 140>, <72 55 90>, <-15 25 160>]}}
      )
     )
     Front End=&AddVectorToPosition(
      Position=&SelectVectorFromArray(Direct Array Index=$Particles.PartitionParticles.Partition, Vector Selection Array=$System.Main Position)
      Vector=&AddVectorToPosition(
       Position=&SelectVectorFromArray(
        Direct Array Index=$Particles.PartitionParticles.Partition
        Vector Selection Array=DataInterface() {propertyValues {FloatData [<15 55 140>, <92 70 90>, <-18 25 160>]}}
       )
       Vector=&SelectVectorFromArray(Direct Array Index=$Particles.PartitionParticles.Partition, Vector Selection Array=$System.Front End Position)
      )
     )
    )
    UsePositionOffset=true
    Write Material Random=true
   )
   &CameraOffset(Camera Offset Amount=10)
   &InitialMeshOrientation(
    Mesh Orientation Mode=::NewEnumerator2
    Rotation=&SelectVectorFromArray(
     Direct Array Index=$Particles.PartitionParticles.Partition
     Vector Selection Array=DataInterface() {propertyValues {FloatData [<-0.05 0 0>, <-0.08 0 0>, <-0.05 0 0>]}}
    )
    Use Rotation Vector=true
   ) {module=InitialMeshOrientation001}
   &InitialMeshOrientation(
    Mesh Orientation Mode=::NewEnumerator2
    Rotation=&FN_FrontEndVector3Modifier(
     Gameplay=&SelectVectorFromArray(
      Direct Array Index=$Particles.PartitionParticles.Partition
      Vector Selection Array=DataInterface() {propertyValues {FloatData [<0 0 0>, <0 0.08 0>, <0 -0.01 0>]}}
     )
     Front End=&SelectVectorFromArray(
      Direct Array Index=$Particles.PartitionParticles.Partition
      Vector Selection Array=DataInterface() {propertyValues {FloatData [<0 0 0>, <0.05 0 0>, <0 -0.01 0>]}}
     )
    )
    Use Rotation Vector=true
   ) {module=InitialMeshOrientation002}
   &DynamicMaterialParameters(
    Index 0 Param 1=&RandomRangeFloat(Minimum=0.1, Maximum=0.5)
    Index 0 Param 2=&RandomRangeFloat(Minimum=0.1, Maximum=1)
    Index 0 Param 3=&Multiply_Float(A=10, B=&FN_FrontEndFloatModifier(Gameplay=1, Front End=0.3))
    Module.Index0Float0Write=true
    Module.Index0Float1Write=true
    Module.Index0Float2Write=true
    Module.Index0Float3Write=true
   )
  ]
  particleUpdateScript [
   &ParticleState(Kill Particles When Lifetime Has Elapsed=false, Let Infinitely Lived Particles Die When Emitter Deactivates=true)
   &ScaleMeshSize(Scale Factor=&VectorFromFloat(Value=&FN_FrontEndFloatModifier(Gameplay=1, Front End=&Multiply_Float(A=1, B=$System.Front End Scale))))
   &SkeletalMeshLocation(
    Bone Sampling Mode=::NewEnumerator4
    MaskSpawnGroup=false
    Orientation Sampling=::NewEnumerator1
    Sampled Position Offset=&Lerp_Vector(
     A=Linked() {linkedVariable {name=Particles.Initial.Position, type=NiagaraPosition}}
     B=&SelectVectorFromArray(
      Direct Array Index=$Particles.PartitionParticles.Partition
      Vector Selection Array=DataInterface() {propertyValues {FloatData [<0 0 200>, <0 0 -100>, <0 0 50>]}}
     )
     Alpha=&Waveform("Waveform [1]"=::NewEnumerator0, "[1] Frequency"=1, "[1] Amplitude Min/Max"=<0 1>, "[1] Scale"=0.1)
    )
    Sampled Position Offset Coordinate Space:ENiagaraSkelMeshTransforms=::NewEnumerator2
    Skeletal Mesh=$System.Skeletal Mesh Reader
   )
   &ScaleColor(
    Scale Alpha=&Multiply_Float(
     A=&FN_FrontEndFloatModifier(Gameplay=1, Front End=1)
     B=&Lerp_Float(A=1, B=$System.TestAlpha, Alpha=&ClampFloat(Float=$User.ParaGlideLeanAlpha))
    )
    ScaleA=true
    ScaleRGB=true
   )
  ]
  rendererClasses [&NiagaraMeshRendererProperties]
  renderers [{rendererClass=&NiagaraMeshRendererProperties, rendererIndex=0}]
 }
 All_FoursTrail {
  emitterUpdateScript [
   &EmitterState(
    Inactive Response=::NewEnumerator1
    Life Cycle Mode=::NewEnumerator1
    Loop Behavior=::NewEnumerator1
    Loop Duration Mode=::NewEnumerator1
    UseLoopDelay=false
   )
   &SpawnBurst_Instantaneous(Spawn Count=4, Use Spawn Probability=false, UseLoopCountLimit=false)
  ]
  particleSpawnScript [
   &PartitionParticles(Number Of Partitions=4)
   &InitializeParticle(
    Color=<rgba 1 1 1 0.5>
    Mesh Scale=&SelectVectorFromArray(
     Direct Array Index=$Particles.PartitionParticles.Partition
     Vector Selection Array=DataInterface() {propertyValues {FloatData [<0.5 0.5 0.5>, <0.5 0.5 0.5>, <0.4 0.4 0.4>, <0.4 0.4 0.4>]}}
    )
    Mesh Scale Mode=::NewEnumerator4
    UsePositionOffset=false
    Write Material Random=true
   )
   &DynamicMaterialParameters(
    Index 0 Param 1=&RandomRangeFloat(Minimum=0.1, Maximum=0.5)
    Index 0 Param 2=&RandomRangeFloat(Minimum=0.1, Maximum=1)
    Index 0 Param 3=&Multiply_Float(A=10, B=&FN_FrontEndFloatModifier(Gameplay=1, Front End=0.3))
    Module.Index0Float0Write=true
    Module.Index0Float1Write=true
    Module.Index0Float2Write=true
    Module.Index0Float3Write=true
   )
  ]
  particleUpdateScript [
   &ParticleState(Kill Particles When Lifetime Has Elapsed=false, Let Infinitely Lived Particles Die When Emitter Deactivates=true)
   &SkeletalMeshLocation(
    Bone / Socket Index=&SelectIntFromArray(Direct Array Index=$Particles.PartitionParticles.Partition, Int Selection Array=DataInterface() {propertyValues {IntData [1,2,3,4]}})
    Bone Sampling Mode=::NewEnumerator4
    MaskSpawnGroup=false
    Orientation Sampling=::NewEnumerator1
    Sampled Position Offset=&SelectVectorFromArray(
     Direct Array Index=$Particles.PartitionParticles.Partition
     Vector Selection Array=DataInterface() {propertyValues {FloatData [<0 0 0>, <0 0 0>, <0 0 0>, <0 0 0>]}}
    )
    Skeletal Mesh=$System.Skeletal Mesh Reader
   )
   &ScaleMeshSize(Scale Factor=&VectorFromFloat(Value=&Lerp_Float(A=0.5, B=0.8, Alpha=&ClampFloat(Float=$User.ParaGlideLeanAlpha)))) {module=ScaleMeshSize001}
   &ScaleColor(
    Scale Alpha=&Multiply_Float(A=&FN_FrontEndFloatModifier(Gameplay=1, Front End=1), B=0.5)
    Scale RGB=&Lerp_Vector(
     A=&Multiply_Vector(
      A=&MakeVectorFromLinearColorRGB(Color=<rgba 1 1 1 1>)
      B=&MakeVectorFromLinearColorRGB(Color=&FNTimeOfDayColor(Night Color=<rgba 0.4 0.55 1 1>, Morning Color=<rgba 1 1 1 1>, Day Color=<rgba 1 1 1 1>, Evening Color=<rgba 1 1 1 1>))
     )
     B=&MakeVectorFromLinearColorRGB(Color=<rgba 3 1 0 1>)
     Alpha=$User.ParaGlideLeanAlpha
    )
    ScaleA=true
    ScaleRGB=true
   )
  ]
  rendererClasses [&NiagaraMeshRendererProperties]
  renderers [{rendererClass=&NiagaraMeshRendererProperties, rendererIndex=0}]
 }
 Eyes {
  emitterUpdateScript [
   &EmitterState(
    Inactive Response=::NewEnumerator1
    Life Cycle Mode=::NewEnumerator1
    Loop Behavior=::NewEnumerator1
    Loop Duration Mode=::NewEnumerator1
    UseLoopDelay=false
   )
   &SpawnBurst_Instantaneous(Spawn Count=3, Use Spawn Probability=false, UseLoopCountLimit=false)
   /SetVariables(Emitter.Texture_Speed=1) {isSetParameters=true}
  ]
  particleSpawnScript [
   &PartitionParticles(Number Of Partitions=3)
   &InitializeParticle(
    Color=$System.Eye Color
    Mesh Scale=&SelectVectorFromArray(
     Direct Array Index=$Particles.PartitionParticles.Partition
     Vector Selection Array=DataInterface() {propertyValues {FloatData [<3.5 30 30>, <3 10 10>, <3 10 10>]}}
    )
    Mesh Scale Mode=::NewEnumerator4
    Position Offset=&FN_FrontEndVector3Modifier(
     Gameplay=&AddVectorToPosition(
      Position=&SelectVectorFromArray(Direct Array Index=$Particles.PartitionParticles.Partition, Vector Selection Array=$System.Main Position)
      Vector=&SelectVectorFromArray(
       Direct Array Index=$Particles.PartitionParticles.Partition
       Vector Selection Array=DataInterface() {propertyValues {FloatData [<10 65 200>, <100 70 142>, <-30 50 215>]}}
      )
     )
     Front End=&AddVectorToPosition(
      Position=&SelectVectorFromArray(Direct Array Index=$Particles.PartitionParticles.Partition, Vector Selection Array=$System.Main Position)
      Vector=&AddVectorToPosition(
       Position=&SelectVectorFromArray(
        Direct Array Index=$Particles.PartitionParticles.Partition
        Vector Selection Array=DataInterface() {propertyValues {FloatData [<12 65 176>, <90 65 140>, <-25 50 215>]}}
       )
       Vector=&SelectVectorFromArray(Direct Array Index=$Particles.PartitionParticles.Partition, Vector Selection Array=$System.Front End Position)
      )
     )
    )
    UsePositionOffset=true
    Write Material Random=true
   )
   &CameraOffset(Camera Offset Amount=&FN_FrontEndFloatModifier(Gameplay=10, Front End=-50))
   /SetVariables(Particles.VisibilityTag=&SelectIntFromArray(Direct Array Index=$Particles.PartitionParticles.Partition, Int Selection Array=DataInterface() {propertyValues {IntData [0,1,2]}})) {isSetParameters=true}
   &InitialMeshOrientation(
    Mesh Orientation Mode=::NewEnumerator2
    Rotation=&FN_FrontEndVector3Modifier(
     Gameplay=&SelectVectorFromArray(
      Direct Array Index=$Particles.PartitionParticles.Partition
      Vector Selection Array=DataInterface() {propertyValues {FloatData [<0 -0.24 0>, <0 -0.17 0>, <0 -0.25 0>]}}
     )
     Front End=&SelectVectorFromArray(
      Direct Array Index=$Particles.PartitionParticles.Partition
      Vector Selection Array=DataInterface() {propertyValues {FloatData [<0 -0.24 0>, <0 -0.25 0>, <0 -0.25 0>]}}
     )
    )
    Use Rotation Vector=true
   ) {module=InitialMeshOrientation001}
   &InitialMeshOrientation(
    Mesh Orientation Mode=::NewEnumerator2
    Rotation=&FN_FrontEndVector3Modifier(
     Gameplay=&SelectVectorFromArray(
      Direct Array Index=$Particles.PartitionParticles.Partition
      Vector Selection Array=DataInterface() {propertyValues {FloatData [<-0.03 0 0>, <-0.1 0 0>, <0 0 0>]}}
     )
     Front End=&SelectVectorFromArray(
      Direct Array Index=$Particles.PartitionParticles.Partition
      Vector Selection Array=DataInterface() {propertyValues {FloatData [<-0.03 0 0>, <-0.01 0 0>, <0 0 0>]}}
     )
    )
    Use Rotation Vector=true
   )
  ]
  particleUpdateScript [
   &ParticleState(Kill Particles When Lifetime Has Elapsed=false, Let Infinitely Lived Particles Die When Emitter Deactivates=true)
   &ScaleMeshSize(Scale Factor=&Multiply_VectorByFloat(
     Vector=<1 1 1>
     Float=&FN_FrontEndFloatModifier(Gameplay=1.5, Front End=&Multiply_Float(A=$System.Front End Scale, B=1.5))
    ))
   &SkeletalMeshLocation(
    Bone Sampling Mode=::NewEnumerator4
    MaskSpawnGroup=false
    Orientation Sampling=::NewEnumerator1
    Sampled Position Offset=&Lerp_Vector(
     A=Linked() {linkedVariable {name=Particles.Initial.Position, type=NiagaraPosition}}
     B=&SelectVectorFromArray(
      Direct Array Index=$Particles.PartitionParticles.Partition
      Vector Selection Array=DataInterface() {propertyValues {FloatData [<0 0 200>, <0 0 -100>, <0 0 50>]}}
     )
     Alpha=&Waveform("Waveform [1]"=::NewEnumerator0, "[1] Frequency"=1, "[1] Amplitude Min/Max"=<0 1>, "[1] Scale"=0.1)
    )
    Sampled Position Offset Coordinate Space:ENiagaraSkelMeshTransforms=::NewEnumerator2
    Skeletal Mesh=$System.Skeletal Mesh Reader
   )
   &ScaleColor(
    Scale Alpha=&FloatFromCurve(
     FloatCurve=DataInterface() {propertyValues {Curve {keys [<at 0 0>,<at 0.1 1>]}}}
     Scale Curve=&Multiply_Float(
      A=&FN_FrontEndFloatModifier(Gameplay=1, Front End=1)
      B=&Lerp_Float(A=$System.TestAlpha, B=1, Alpha=&Subtract_Float(A=$User.ParaGlideLeanAlpha, B=$System.Threshold))
     )
    )
    Scale RGB=&VectorFromFloat(Value=$System.Brightness Scale)
    ScaleA=true
    ScaleRGB=true
   )
  ]
  rendererClasses [&NiagaraMeshRendererProperties]
  renderers [
   {rendererClass=&NiagaraMeshRendererProperties, rendererIndex=0}
   {rendererClass=&NiagaraMeshRendererProperties, rendererIndex=1}
   {rendererClass=&NiagaraMeshRendererProperties, rendererIndex=2}
  ]
 }
 EyesTrai {
  emitterUpdateScript [
   &EmitterState(
    Inactive Response=::NewEnumerator1
    Life Cycle Mode=::NewEnumerator1
    Loop Behavior=::NewEnumerator1
    Loop Duration Mode=::NewEnumerator1
    UseLoopDelay=false
   )
   &SpawnBurst_Instantaneous(Spawn Count=$System.Test Num, Use Spawn Probability=false, UseLoopCountLimit=false)
  ]
  particleSpawnScript [
   &PartitionParticles(Number Of Partitions=3)
   &InitializeParticle(
    Color=<rgba 1 1 1 80>
    Mesh Scale=&SelectVectorFromArray(
     Direct Array Index=$Particles.PartitionParticles.Partition
     Vector Selection Array=DataInterface() {propertyValues {FloatData [<5 5 5>, <4 -4 6>, <5 5 6>]}}
    )
    Mesh Scale Mode=::NewEnumerator4
    Position Offset=&FN_FrontEndVector3Modifier(
     Gameplay=&AddVectorToPosition(
      Position=&SelectVectorFromArray(Direct Array Index=$Particles.PartitionParticles.Partition, Vector Selection Array=$System.Main Position)
      Vector=&SelectVectorFromArray(
       Direct Array Index=$Particles.PartitionParticles.Partition
       Vector Selection Array=DataInterface() {propertyValues {FloatData [<0 45 100>, <85 40 60>, <-15 15 100>]}}
      )
     )
     Front End=&AddVectorToPosition(
      Position=&SelectVectorFromArray(Direct Array Index=$Particles.PartitionParticles.Partition, Vector Selection Array=$System.Main Position)
      Vector=&AddVectorToPosition(
       Position=&SelectVectorFromArray(
        Direct Array Index=$Particles.PartitionParticles.Partition
        Vector Selection Array=DataInterface() {propertyValues {FloatData [<15 35 120>, <105 70 90>, <-5 16 150>]}}
       )
       Vector=&SelectVectorFromArray(Direct Array Index=$Particles.PartitionParticles.Partition, Vector Selection Array=$System.Front End Position)
      )
     )
    )
    UsePositionOffset=true
    Write Material Random=true
   )
   &CameraOffset(Camera Offset Amount=&FN_FrontEndFloatModifier(Gameplay=10, Front End=-50))
   &InitialMeshOrientation(
    Mesh Orientation Mode=::NewEnumerator2
    Rotation=&SelectVectorFromArray(Direct Array Index=$Particles.PartitionParticles.Partition, Vector Selection Array=$System.Trail RotationZ)
    Use Rotation Vector=true
   )
   &InitialMeshOrientation(
    Mesh Orientation Mode=::NewEnumerator2
    Rotation=&FN_FrontEndVector3Modifier(
     Gameplay=&SelectVectorFromArray(Direct Array Index=$Particles.PartitionParticles.Partition, Vector Selection Array=$System.Trail RotationY)
     Front End=&SelectVectorFromArray(Direct Array Index=$Particles.PartitionParticles.Partition, Vector Selection Array=$System.Front End Trail Y Rotation)
    )
    Use Rotation Vector=true
   ) {module=InitialMeshOrientation001}
   &InitialMeshOrientation(
    Mesh Orientation Mode=::NewEnumerator2
    Rotation=&FN_FrontEndVector3Modifier(
     Gameplay=&SelectVectorFromArray(Direct Array Index=$Particles.PartitionParticles.Partition, Vector Selection Array=$System.Trail RotationX)
     Front End=&SelectVectorFromArray(Direct Array Index=$Particles.PartitionParticles.Partition, Vector Selection Array=$System.Front End Trail X Rotation)
    )
    Use Rotation Vector=true
   ) {module=InitialMeshOrientation002}
   &InitialMeshOrientation(
    Mesh Orientation Mode=::NewEnumerator2
    Rotation=&SelectVectorFromArray(
     Direct Array Index=$Particles.PartitionParticles.Partition
     Vector Selection Array=DataInterface() {propertyValues {FloatData [<0 -0.035 0>, <0 -0.05 0>, <0 -0.04 0>]}}
    )
    Use Rotation Vector=true
   ) {module=InitialMeshOrientation003}
   &DynamicMaterialParameters(
    Index 0 Param 1=1
    Index 0 Param 2=&RandomRangeFloat(Minimum=0.8, Maximum=1)
    Index 0 Param 3=&SelectFloatFromArray(Direct Array Index=$Particles.PartitionParticles.Partition, Float Selection Array=DataInterface() {propertyValues {FloatData [400,300,500]}})
    Index 0 Param 4=&SelectFloatFromArray(Direct Array Index=$Particles.PartitionParticles.Partition, Float Selection Array=DataInterface() {propertyValues {FloatData [1,1,2]}})
    Module.Index0Float0Write=true
    Module.Index0Float1Write=true
    Module.Index0Float2Write=true
    Module.Index0Float3Write=true
   )
  ]
  particleUpdateScript [
   &ParticleState(Kill Particles When Lifetime Has Elapsed=false, Let Infinitely Lived Particles Die When Emitter Deactivates=true)
   &ScaleMeshSize(Scale Factor=&Multiply_VectorByFloat(
     Vector=<2 2 1.5>
     Float=&FN_FrontEndFloatModifier(Gameplay=0.7, Front End=&Multiply_Float(A=$System.Front End Scale, B=0.7))
    ))
   &SkeletalMeshLocation(
    Bone Sampling Mode=::NewEnumerator4
    MaskSpawnGroup=false
    Orientation Sampling=::NewEnumerator1
    Sampled Position Offset=&Lerp_Vector(
     A=Linked() {linkedVariable {name=Particles.Initial.Position, type=NiagaraPosition}}
     B=&SelectVectorFromArray(
      Direct Array Index=$Particles.PartitionParticles.Partition
      Vector Selection Array=DataInterface() {propertyValues {FloatData [<0 0 200>, <0 0 -100>, <0 0 50>]}}
     )
     Alpha=&Waveform("Waveform [1]"=::NewEnumerator0, "[1] Frequency"=1, "[1] Amplitude Min/Max"=<0 1>, "[1] Scale"=0.1)
    )
    Sampled Position Offset Coordinate Space:ENiagaraSkelMeshTransforms=::NewEnumerator2
    Skeletal Mesh=$System.Skeletal Mesh Reader
   )
   /SetVariables(Particles.MaterialRandom:NiagaraFloat=&RandomRangeFloat(Minimum=0, Maximum=1)) {isSetParameters=true, module=SetVariables_#6}
   &ScaleColor(
    Scale Alpha=&Multiply_Float(
     A=&FN_FrontEndFloatModifier(Gameplay=0.72, Front End=0.2)
     B=&Lerp_Float(A=$System.TestAlpha, B=1, Alpha=&Subtract_Float(A=$User.ParaGlideLeanAlpha, B=$System.Threshold))
    )
    Scale RGB=&VectorFromFloat(Value=$System.Brightness Scale)
    ScaleA=true
    ScaleRGB=true
   )
  ]
  rendererClasses [&NiagaraMeshRendererProperties]
  renderers [{rendererClass=&NiagaraMeshRendererProperties, rendererIndex=0}]
 }
 MainTrail {
  emitterUpdateScript [
   &EmitterState(
    Inactive Response=::NewEnumerator1
    Life Cycle Mode=::NewEnumerator1
    Loop Behavior=::NewEnumerator1
    Loop Duration Mode=::NewEnumerator1
    UseLoopDelay=false
   )
   &SpawnBurst_Instantaneous(Spawn Count=$System.Test Num, Use Spawn Probability=false, UseLoopCountLimit=false)
  ]
  particleSpawnScript [
   &PartitionParticles(Number Of Partitions=3)
   &InitializeParticle(
    Color=<rgba 1 1 1 100>
    Mesh Scale=&SelectVectorFromArray(
     Direct Array Index=$Particles.PartitionParticles.Partition
     Vector Selection Array=DataInterface() {propertyValues {FloatData [<3 3 6>, <3 3 5>, <4 4 7>]}}
    )
    Mesh Scale Mode=::NewEnumerator4
    Position Offset=&FN_FrontEndVector3Modifier(
     Gameplay=&AddVectorToPosition(
      Position=&SelectVectorFromArray(Direct Array Index=$Particles.PartitionParticles.Partition, Vector Selection Array=$System.Main Position)
      Vector=&SelectVectorFromArray(
       Direct Array Index=$Particles.PartitionParticles.Partition
       Vector Selection Array=DataInterface() {propertyValues {FloatData [<0 0 5>, <-10 -10 5>, <0 -2 0>]}}
      )
     )
     Front End=&AddVectorToPosition(
      Position=&SelectVectorFromArray(Direct Array Index=$Particles.PartitionParticles.Partition, Vector Selection Array=$System.Main Position)
      Vector=&AddVectorToPosition(
       Position=&SelectVectorFromArray(
        Direct Array Index=$Particles.PartitionParticles.Partition
        Vector Selection Array=DataInterface() {propertyValues {FloatData [<10 20 45>, <72 50 -15>, <2 6 40>]}}
       )
       Vector=&SelectVectorFromArray(Direct Array Index=$Particles.PartitionParticles.Partition, Vector Selection Array=$System.Front End Position)
      )
     )
    )
    UsePositionOffset=true
    Write Material Random=true
   )
   &CameraOffset(Camera Offset Amount=&FN_FrontEndFloatModifier(Gameplay=10, Front End=-10))
   &InitialMeshOrientation(
    Mesh Orientation Mode=::NewEnumerator2
    Rotation=&SelectVectorFromArray(Direct Array Index=$Particles.PartitionParticles.Partition, Vector Selection Array=$System.Trail RotationZ)
    Use Rotation Vector=true
   )
   &InitialMeshOrientation(
    Mesh Orientation Mode=::NewEnumerator2
    Rotation=&FN_FrontEndVector3Modifier(
     Gameplay=&SelectVectorFromArray(Direct Array Index=$Particles.PartitionParticles.Partition, Vector Selection Array=$System.Trail RotationY)
     Front End=&SelectVectorFromArray(Direct Array Index=$Particles.PartitionParticles.Partition, Vector Selection Array=$System.Front End Trail Y Rotation)
    )
    Use Rotation Vector=true
   ) {module=InitialMeshOrientation001}
   &InitialMeshOrientation(
    Mesh Orientation Mode=::NewEnumerator2
    Rotation=&FN_FrontEndVector3Modifier(
     Gameplay=&SelectVectorFromArray(Direct Array Index=$Particles.PartitionParticles.Partition, Vector Selection Array=$System.Trail RotationX)
     Front End=&SelectVectorFromArray(
      Direct Array Index=$Particles.PartitionParticles.Partition
      Vector Selection Array=DataInterface() {propertyValues {FloatData [<-0.03 0 0>, <-0.03 0 0>, <-0.01 0 0>]}}
     )
    )
    Use Rotation Vector=true
   ) {module=InitialMeshOrientation002}
   &DynamicMaterialParameters(
    Index 0 Param 1=1
    Index 0 Param 2=&RandomRangeFloat(Minimum=0.8, Maximum=1)
    Index 0 Param 3=&SelectFloatFromArray(Direct Array Index=$Particles.PartitionParticles.Partition, Float Selection Array=DataInterface() {propertyValues {FloatData [400,450,450]}})
    Index 0 Param 4=&SelectFloatFromArray(Direct Array Index=$Particles.PartitionParticles.Partition, Float Selection Array=DataInterface() {propertyValues {FloatData [1.5,1.5,1.5]}})
    Index 3 Param 4=20
    Module.Index0Float0Write=true
    Module.Index0Float1Write=true
    Module.Index0Float2Write=true
    Module.Index0Float3Write=true
    Module.Index3Float0Write=true
    Module.Index3Float1Write=true
    Module.Index3Float2Write=true
    Module.Index3Float3Write=true
    Write Parameter Index 3=::NewEnumerator1
   )
  ]
  particleUpdateScript [
   &ParticleState(Kill Particles When Lifetime Has Elapsed=false, Let Infinitely Lived Particles Die When Emitter Deactivates=true)
   &ScaleMeshSize(Scale Factor=&Multiply_VectorByFloat(
     Vector=<1 1 1>
     Float=&FN_FrontEndFloatModifier(Gameplay=1.8, Front End=&Multiply_Float(A=$System.Front End Scale, B=1.8))
    ))
   &SkeletalMeshLocation(
    Bone Sampling Mode=::NewEnumerator4
    MaskSpawnGroup=false
    Orientation Sampling=::NewEnumerator1
    Sampled Position Offset=&Lerp_Vector(
     A=Linked() {linkedVariable {name=Particles.Initial.Position, type=NiagaraPosition}}
     B=&SelectVectorFromArray(
      Direct Array Index=$Particles.PartitionParticles.Partition
      Vector Selection Array=DataInterface() {propertyValues {FloatData [<0 0 200>, <0 0 -100>, <0 0 50>]}}
     )
     Alpha=&Waveform("Waveform [1]"=::NewEnumerator0, "[1] Frequency"=1, "[1] Amplitude Min/Max"=<0 1>, "[1] Scale"=0.1)
    )
    Sampled Position Offset Coordinate Space:ENiagaraSkelMeshTransforms=::NewEnumerator2
    Skeletal Mesh=$System.Skeletal Mesh Reader
   )
   /SetVariables(Particles.MaterialRandom:NiagaraFloat=&RandomRangeFloat(Minimum=0, Maximum=1)) {isSetParameters=true, module=SetVariables_#6}
   &ScaleColor(
    Scale Alpha=&Multiply_Float(
     A=&FN_FrontEndFloatModifier(Gameplay=1.8, Front End=2)
     B=&Lerp_Float(A=$System.TestAlpha, B=1, Alpha=&Subtract_Float(A=$User.ParaGlideLeanAlpha, B=$System.Threshold))
    )
    Scale RGB=&VectorFromFloat(Value=$System.Brightness Scale)
    ScaleA=true
    ScaleRGB=true
   )
  ]
  rendererClasses [&NiagaraMeshRendererProperties]
  renderers [{rendererClass=&NiagaraMeshRendererProperties, rendererIndex=0}]
 }
 MainTrail_L {
  emitterUpdateScript [
   &EmitterState(
    Inactive Response=::NewEnumerator1
    Life Cycle Mode=::NewEnumerator1
    Loop Behavior=::NewEnumerator1
    Loop Duration Mode=::NewEnumerator1
    UseLoopDelay=false
   )
   &SpawnBurst_Instantaneous(Spawn Count=$System.Test Num, Use Spawn Probability=false, UseLoopCountLimit=false)
  ]
  particleSpawnScript [
   &PartitionParticles(Number Of Partitions=3)
   &InitializeParticle(
    Color=<rgba 1 1 1 80>
    Mesh Scale=&SelectVectorFromArray(
     Direct Array Index=$Particles.PartitionParticles.Partition
     Vector Selection Array=DataInterface() {propertyValues {FloatData [<3 3 5>, <-4 4 5>, <-3 -3 4>]}}
    )
    Mesh Scale Mode=::NewEnumerator4
    Position Offset=&FN_FrontEndVector3Modifier(
     Gameplay=&AddVectorToPosition(
      Position=&SelectVectorFromArray(Direct Array Index=$Particles.PartitionParticles.Partition, Vector Selection Array=$System.Main Position)
      Vector=&SelectVectorFromArray(
       Direct Array Index=$Particles.PartitionParticles.Partition
       Vector Selection Array=DataInterface() {propertyValues {FloatData [<-25 5 0>, <0 0 50>, <-25 15 20>]}}
      )
     )
     Front End=&AddVectorToPosition(
      Position=&SelectVectorFromArray(Direct Array Index=$Particles.PartitionParticles.Partition, Vector Selection Array=$System.Main Position)
      Vector=&AddVectorToPosition(
       Position=&SelectVectorFromArray(
        Direct Array Index=$Particles.PartitionParticles.Partition
        Vector Selection Array=DataInterface() {propertyValues {FloatData [<-5 18 30>, <60 55 0>, <-20 10 60>]}}
       )
       Vector=&SelectVectorFromArray(Direct Array Index=$Particles.PartitionParticles.Partition, Vector Selection Array=$System.Front End Position)
      )
     )
    )
    UsePositionOffset=true
    Write Material Random=true
   )
   &CameraOffset(Camera Offset Amount=&FN_FrontEndFloatModifier(Gameplay=10, Front End=-50))
   &InitialMeshOrientation(
    Mesh Orientation Mode=::NewEnumerator2
    Rotation=&SelectVectorFromArray(Direct Array Index=$Particles.PartitionParticles.Partition, Vector Selection Array=$System.Trail RotationZ)
    Use Rotation Vector=true
   )
   &InitialMeshOrientation(
    Mesh Orientation Mode=::NewEnumerator2
    Rotation=&FN_FrontEndVector3Modifier(
     Gameplay=&AddVectorToPosition(
      Position=&SelectVectorFromArray(Direct Array Index=$Particles.PartitionParticles.Partition, Vector Selection Array=$System.Trail RotationY)
      Vector=&SelectVectorFromArray(
       Direct Array Index=$Particles.PartitionParticles.Partition
       Vector Selection Array=DataInterface() {propertyValues {FloatData [<0 0.02 0>, <0 0 0>, <0 0 0>]}}
      )
     )
     Front End=&SelectVectorFromArray(Direct Array Index=$Particles.PartitionParticles.Partition, Vector Selection Array=$System.Front End Trail Y Rotation)
    )
    Use Rotation Vector=true
   ) {module=InitialMeshOrientation001}
   &InitialMeshOrientation(
    Mesh Orientation Mode=::NewEnumerator2
    Rotation=&FN_FrontEndVector3Modifier(
     Gameplay=&SelectVectorFromArray(Direct Array Index=$Particles.PartitionParticles.Partition, Vector Selection Array=$System.Trail RotationX)
     Front End=&SelectVectorFromArray(Direct Array Index=$Particles.PartitionParticles.Partition, Vector Selection Array=$System.Front End Trail X Rotation)
    )
    Use Rotation Vector=true
   ) {module=InitialMeshOrientation002}
   &InitialMeshOrientation(
    Mesh Orientation Mode=::NewEnumerator2
    Rotation=&SelectVectorFromArray(
     Direct Array Index=$Particles.PartitionParticles.Partition
     Vector Selection Array=DataInterface() {propertyValues {FloatData [<0.01 0.03 0>, <0 0.05 0>, <0 0.01 0>]}}
    )
    Use Rotation Vector=true
   ) {module=InitialMeshOrientation003}
   &DynamicMaterialParameters(
    Index 0 Param 1=1
    Index 0 Param 2=&RandomRangeFloat(Minimum=0.8, Maximum=1)
    Index 0 Param 3=&SelectFloatFromArray(Direct Array Index=$Particles.PartitionParticles.Partition, Float Selection Array=DataInterface() {propertyValues {FloatData [450,600,500]}})
    Index 0 Param 4=&SelectFloatFromArray(Direct Array Index=$Particles.PartitionParticles.Partition, Float Selection Array=DataInterface() {propertyValues {FloatData [3.75,3,3]}})
    Index 3 Param 4=20
    Module.Index0Float0Write=true
    Module.Index0Float1Write=true
    Module.Index0Float2Write=true
    Module.Index0Float3Write=true
    Module.Index3Float0Write=true
    Module.Index3Float1Write=true
    Module.Index3Float2Write=true
    Module.Index3Float3Write=true
    Write Parameter Index 3=::NewEnumerator1
   )
  ]
  particleUpdateScript [
   &ParticleState(Kill Particles When Lifetime Has Elapsed=false, Let Infinitely Lived Particles Die When Emitter Deactivates=true)
   &ScaleMeshSize(Scale Factor=&Multiply_VectorByFloat(
     Vector=<1 1 1.5>
     Float=&FN_FrontEndFloatModifier(Gameplay=1.8, Front End=&Multiply_Float(A=$System.Front End Scale, B=1.8))
    ))
   &SkeletalMeshLocation(
    Bone Sampling Mode=::NewEnumerator4
    MaskSpawnGroup=false
    Orientation Sampling=::NewEnumerator1
    Sampled Position Offset=&Lerp_Vector(
     A=Linked() {linkedVariable {name=Particles.Initial.Position, type=NiagaraPosition}}
     B=&SelectVectorFromArray(
      Direct Array Index=$Particles.PartitionParticles.Partition
      Vector Selection Array=DataInterface() {propertyValues {FloatData [<0 0 200>, <0 0 -100>, <0 0 50>]}}
     )
     Alpha=&Waveform("Waveform [1]"=::NewEnumerator0, "[1] Frequency"=1, "[1] Amplitude Min/Max"=<0 1>, "[1] Scale"=0.1)
    )
    Sampled Position Offset Coordinate Space:ENiagaraSkelMeshTransforms=::NewEnumerator2
    Skeletal Mesh=$System.Skeletal Mesh Reader
   )
   /SetVariables(Particles.MaterialRandom:NiagaraFloat=&RandomRangeFloat(Minimum=0, Maximum=1)) {isSetParameters=true, module=SetVariables_#6}
   &ScaleColor(
    Scale Alpha=&Multiply_Float(
     A=&FN_FrontEndFloatModifier(Gameplay=1.8, Front End=1.5)
     B=&Lerp_Float(A=$System.TestAlpha, B=1, Alpha=&Subtract_Float(A=$User.ParaGlideLeanAlpha, B=$System.Threshold))
    )
    Scale RGB=&VectorFromFloat(Value=$System.Brightness Scale)
    ScaleA=true
    ScaleRGB=true
   )
  ]
  rendererClasses [&NiagaraMeshRendererProperties]
  renderers [{rendererClass=&NiagaraMeshRendererProperties, rendererIndex=0}]
 }
 MainTrail_R {
  emitterUpdateScript [
   &EmitterState(
    Inactive Response=::NewEnumerator1
    Life Cycle Mode=::NewEnumerator1
    Loop Behavior=::NewEnumerator1
    Loop Duration Mode=::NewEnumerator1
    UseLoopDelay=false
   )
   &SpawnBurst_Instantaneous(Spawn Count=$System.Test Num, Use Spawn Probability=false, UseLoopCountLimit=false)
  ]
  particleSpawnScript [
   &PartitionParticles(Number Of Partitions=3)
   &InitializeParticle(
    Color=<rgba 1 1 1 80>
    Mesh Scale=&SelectVectorFromArray(
     Direct Array Index=$Particles.PartitionParticles.Partition
     Vector Selection Array=DataInterface() {propertyValues {FloatData [<-1 1 5>, <-3 3 5>, <-2 2 5>]}}
    )
    Mesh Scale Mode=::NewEnumerator4
    Position Offset=&FN_FrontEndVector3Modifier(
     Gameplay=&AddVectorToPosition(
      Position=&SelectVectorFromArray(Direct Array Index=$Particles.PartitionParticles.Partition, Vector Selection Array=$System.Main Position)
      Vector=&SelectVectorFromArray(
       Direct Array Index=$Particles.PartitionParticles.Partition
       Vector Selection Array=DataInterface() {propertyValues {FloatData [<10 10 20>, <20 15 30>, <10 5 30>]}}
      )
     )
     Front End=&AddVectorToPosition(
      Position=&SelectVectorFromArray(Direct Array Index=$Particles.PartitionParticles.Partition, Vector Selection Array=$System.Main Position)
      Vector=&AddVectorToPosition(
       Position=&SelectVectorFromArray(
        Direct Array Index=$Particles.PartitionParticles.Partition
        Vector Selection Array=DataInterface() {propertyValues {FloatData [<15 16 50>, <82 53 0>, <7 5 55>]}}
       )
       Vector=&SelectVectorFromArray(Direct Array Index=$Particles.PartitionParticles.Partition, Vector Selection Array=$System.Front End Position)
      )
     )
    )
    UsePositionOffset=true
    Write Material Random=true
   )
   &CameraOffset(Camera Offset Amount=&FN_FrontEndFloatModifier(Gameplay=10, Front End=-50))
   &InitialMeshOrientation(
    Mesh Orientation Mode=::NewEnumerator2
    Rotation=&SelectVectorFromArray(Direct Array Index=$Particles.PartitionParticles.Partition, Vector Selection Array=$System.Trail RotationZ)
    Use Rotation Vector=true
   )
   &InitialMeshOrientation(
    Mesh Orientation Mode=::NewEnumerator2
    Rotation=&FN_FrontEndVector3Modifier(
     Gameplay=&SelectVectorFromArray(Direct Array Index=$Particles.PartitionParticles.Partition, Vector Selection Array=$System.Trail RotationY)
     Front End=&SelectVectorFromArray(Direct Array Index=$Particles.PartitionParticles.Partition, Vector Selection Array=$System.Front End Trail Y Rotation)
    )
    Use Rotation Vector=true
   ) {module=InitialMeshOrientation001}
   &InitialMeshOrientation(
    Mesh Orientation Mode=::NewEnumerator2
    Rotation=&FN_FrontEndVector3Modifier(
     Gameplay=&SelectVectorFromArray(Direct Array Index=$Particles.PartitionParticles.Partition, Vector Selection Array=$System.Trail RotationX)
     Front End=&SelectVectorFromArray(Direct Array Index=$Particles.PartitionParticles.Partition, Vector Selection Array=$System.Front End Trail X Rotation)
    )
    Use Rotation Vector=true
   ) {module=InitialMeshOrientation002}
   &InitialMeshOrientation(
    Mesh Orientation Mode=::NewEnumerator2
    Rotation=&SelectVectorFromArray(
     Direct Array Index=$Particles.PartitionParticles.Partition
     Vector Selection Array=DataInterface() {propertyValues {FloatData [<-0.04 -0.04 0>, <0 -0.02 0>, <0 -0.05 0>]}}
    )
    Use Rotation Vector=true
   ) {module=InitialMeshOrientation003}
   &DynamicMaterialParameters(
    Index 0 Param 1=1
    Index 0 Param 2=&RandomRangeFloat(Minimum=0.8, Maximum=1)
    Index 0 Param 3=&SelectFloatFromArray(Direct Array Index=$Particles.PartitionParticles.Partition, Float Selection Array=DataInterface() {propertyValues {FloatData [600,400,600]}})
    Index 0 Param 4=&SelectFloatFromArray(Direct Array Index=$Particles.PartitionParticles.Partition, Float Selection Array=DataInterface() {propertyValues {FloatData [1.5,3,3]}})
    Index 3 Param 4=60
    Module.Index0Float0Write=true
    Module.Index0Float1Write=true
    Module.Index0Float2Write=true
    Module.Index0Float3Write=true
    Module.Index3Float0Write=true
    Module.Index3Float1Write=true
    Module.Index3Float2Write=true
    Module.Index3Float3Write=true
    Write Parameter Index 3=::NewEnumerator1
   )
  ]
  particleUpdateScript [
   &ParticleState(Kill Particles When Lifetime Has Elapsed=false, Let Infinitely Lived Particles Die When Emitter Deactivates=true)
   &ScaleMeshSize(Scale Factor=&Multiply_VectorByFloat(
     Vector=<1 1 1.5>
     Float=&FN_FrontEndFloatModifier(Gameplay=1.3, Front End=&Multiply_Float(A=$System.Front End Scale, B=1.3))
    ))
   &SkeletalMeshLocation(
    Bone Sampling Mode=::NewEnumerator4
    MaskSpawnGroup=false
    Orientation Sampling=::NewEnumerator1
    Sampled Position Offset=&Lerp_Vector(
     A=Linked() {linkedVariable {name=Particles.Initial.Position, type=NiagaraPosition}}
     B=&SelectVectorFromArray(
      Direct Array Index=$Particles.PartitionParticles.Partition
      Vector Selection Array=DataInterface() {propertyValues {FloatData [<0 0 200>, <0 0 -100>, <0 0 50>]}}
     )
     Alpha=&Waveform("Waveform [1]"=::NewEnumerator0, "[1] Frequency"=1, "[1] Amplitude Min/Max"=<0 1>, "[1] Scale"=0.1)
    )
    Sampled Position Offset Coordinate Space:ENiagaraSkelMeshTransforms=::NewEnumerator2
    Skeletal Mesh=$System.Skeletal Mesh Reader
   )
   /SetVariables(Particles.MaterialRandom:NiagaraFloat=&RandomRangeFloat(Minimum=0, Maximum=1)) {isSetParameters=true, module=SetVariables_#6}
   &ScaleColor(
    Scale Alpha=&Multiply_Float(
     A=&FN_FrontEndFloatModifier(Gameplay=1.56, Front End=2)
     B=&Lerp_Float(A=$System.TestAlpha, B=1, Alpha=&Subtract_Float(A=$User.ParaGlideLeanAlpha, B=$System.Threshold))
    )
    Scale RGB=&VectorFromFloat(Value=$System.Brightness Scale)
    ScaleA=true
    ScaleRGB=true
   )
  ]
  rendererClasses [&NiagaraMeshRendererProperties]
  renderers [{rendererClass=&NiagaraMeshRendererProperties, rendererIndex=0}]
 }
 Sparks {
  emitterUpdateScript [
   &EmitterState(Inactive Response=::NewEnumerator1, Life Cycle Mode=::NewEnumerator1, Loop Duration=1, UseLoopDelay=false)
   &SpawnBurst_Instantaneous(Spawn Count=5, Use Spawn Probability=false, UseLoopCountLimit=false)
   &SpawnRate(Spawn Probability=&Lerp_Float(A=0, B=1, Alpha=$User.ParaGlideLeanAlpha), SpawnRate=70, Use Spawn Probability=true)
  ]
  particleSpawnScript [
   &PartitionParticles(Number Of Partitions=3)
   &InitializeParticle(
    Color=<rgba 10 0.8333 0 1>
    Lifetime Max=1.5
    Lifetime Mode=::NewEnumerator1
    Sprite Size Max=<2 10>
    Sprite Size Min=<1 5>
    Sprite Size Mode=::NewEnumerator2
    UsePositionOffset=false
    Write Material Random=true
   )
   &SkeletalMeshLocation(
    Bone Sampling Mode=::NewEnumerator4
    MaskSpawnGroup=false
    Orientation Sampling=::NewEnumerator1
    Sampled Position Offset=&FN_FrontEndVector3Modifier(
     Gameplay=&AddVectorToPosition(
      Position=&SelectVectorFromArray(Direct Array Index=$Particles.PartitionParticles.Partition, Vector Selection Array=$System.Main Position)
      Vector=&SelectVectorFromArray(
       Direct Array Index=$Particles.PartitionParticles.Partition
       Vector Selection Array=DataInterface() {propertyValues {FloatData [<17 15 50>, <52 47 20>, <-10 12 90>]}}
      )
     )
     Front End=&AddVectorToPosition(
      Position=&SelectVectorFromArray(Direct Array Index=$Particles.PartitionParticles.Partition, Vector Selection Array=$System.Main Position)
      Vector=&AddVectorToPosition(
       Position=&SelectVectorFromArray(
        Direct Array Index=$Particles.PartitionParticles.Partition
        Vector Selection Array=DataInterface() {propertyValues {FloatData [<17 15 50>, <80 60 20>, <-10 12 90>]}}
       )
       Vector=&SelectVectorFromArray(Direct Array Index=$Particles.PartitionParticles.Partition, Vector Selection Array=$System.Front End Position)
      )
     )
    )
    Sampled Position Offset Coordinate Space:ENiagaraSkelMeshTransforms=::NewEnumerator2
    Skeletal Mesh=$System.Skeletal Mesh Reader
   )
   &ShapeLocation(MaskSpawnGroup=false, Override Seed=false, Sphere Radius=30, UseOwnerScale=true)
   &AddVelocity(
    Cone Angle=30
    Cone Axis=<0 0 1>
    Module.UseConeFalloff=false
    Module.UseVelDistribution=false
    Override Seed=false
    Velocity Mode=::NewEnumerator2
    Velocity Speed=&RandomRangeFloat(Minimum=600, Maximum=800)
   )
   &AddVelocity(Velocity=<0 100 0>) {module=AddVelocity001}
  ]
  particleUpdateScript [
   &ParticleState()
   &FN_Sprite_Attributes(
    Color=&MakeLinearColorFromVectorAndFloat(
     "Vector (RGB)"=&MakeVectorFromLinearColorRGB(Color=&MultiplyLinearColor(Linear Color=<rgba 1 0.0833 0 1>, Scale Factor=&MakeVector4_FromFloat(Float=1)))
     "Float (Alpha)"=&Multiply_Float(
      A=&RandomRangeFloat(Minimum=0.15, Maximum=0.05)
      B=&Multiply_Float(
       A=&FN_FrontEndFloatModifier(Gameplay=1, Front End=0.6)
       B=&Lerp_Float(A=$System.TestAlpha, B=1, Alpha=&Subtract_Float(A=$User.ParaGlideLeanAlpha, B=0.2))
      )
     )
    )
    Create Color Attribute=true
    Create SpriteSize Attribute=true
    Module Output:ENiagara_AttributeSamplingApplyOutput=::NewEnumerator1
    SpriteSize=&Multiply_Vector2D_ByFloat(
     Float=&FloatFromCurve(
      FloatCurve=DataInterface() {propertyValues {Curve {keys [
          <at 0 0 RCIM_Cubic>
          <at 0.2 1.3 RCIM_Cubic>
          {arriveTangent=-0.1423, interpMode=RCIM_Cubic, leaveTangent=-0.1423, tangentMode=RCTM_User, time=1, value=0}
         ]}}}
      Scale Curve=15
     )
     Vector2D=&Multiply_Vector2D_ByFloat(
      Float=&FN_FrontEndFloatModifier(Gameplay=1, Front End=&Multiply_Float(A=2, B=$System.Front End Scale))
      Vector2D=<0.5 1>
     )
    )
   )
   &ScaleColor(
    Scale Alpha=&Multiply_Float(
     A=&FN_FrontEndFloatModifier(Gameplay=1, Front End=0.6)
     B=&Lerp_Float(A=$System.TestAlpha, B=1, Alpha=&Subtract_Float(A=$User.ParaGlideLeanAlpha, B=$System.Threshold))
    )
    ScaleA=true
    ScaleRGB=true
   )
   &ScaleSpriteSize(Scale Factor=&Vector2DFromFloat(Value=&FloatFromCurve(
      FloatCurve=DataInterface() {propertyValues {Curve {keys [
          <at 0 0 RCIM_Cubic>
          {interpMode=RCIM_Cubic, leaveTangent=-3.0296, tangentMode=RCTM_Break, time=0.2, value=1.2}
          {arriveTangent=-1.8107, interpMode=RCIM_Cubic, leaveTangent=-1.8107, tangentMode=RCTM_User, time=1, value=0}
         ]}}}
      Scale Curve=&FN_FrontEndFloatModifier(Gameplay=0.4, Front End=&Multiply_Float(A=0.4, B=$System.Front End Scale))
     ))) {module=ScaleSpriteSize001}
   &Drag(
    Drag=&FloatFromCurve(
     FloatCurve=DataInterface() {propertyValues {Curve {keys [{arriveTangent=8.8753, interpMode=RCIM_Cubic, leaveTangent=8.8753, tangentMode=RCTM_User, time=-0.003, value=-0.0388}, <at 0.5 2>]}}}
     Scale Curve=1.5
    )
    Use Linear Drag=true
    Use Rotational Drag=true
   )
   &CurlNoiseForce(
    Bias Noise=false
    Noise Frequency=10
    Noise Strength=&RandomRangeFloat(Minimum=1000, Maximum=30)
    Pan Noise=false
   )
   &SolveForcesAndVelocity(Clamp Velocity=false, Limit Acceleration=false)
  ]
  rendererClasses [&NiagaraSpriteRendererProperties]
  renderers [
   {rendererClass=&NiagaraSpriteRendererProperties}
   {rendererClass=&NiagaraSpriteRendererProperties, rendererIndex=1}
  ]
 }
 Sword {
  emitterUpdateScript [
   &EmitterState(
    Inactive Response=::NewEnumerator1
    Life Cycle Mode=::NewEnumerator1
    Loop Behavior=::NewEnumerator1
    Loop Duration Mode=::NewEnumerator1
    UseLoopDelay=false
   )
   &SpawnBurst_Instantaneous(Spawn Count=3, Use Spawn Probability=false, UseLoopCountLimit=false)
  ]
  particleSpawnScript [
   &PartitionParticles(Number Of Partitions=3)
   &InitializeParticle(
    Color=&Multiply_LinearColorByFloat(Linear Color=<rgba 1 1 1 1>, Float=&FN_FrontEndFloatModifier(Gameplay=2, Front End=3))
    Mesh Scale=&Multiply_VectorByFloat(
     Vector=<1 1 1>
     Float=&FN_FrontEndFloatModifier(Gameplay=1.3, Front End=&Multiply_Float(A=1.3, B=$System.Front End Scale))
    )
    Mesh Scale Mode=::NewEnumerator4
    Position Offset=&FN_FrontEndVector3Modifier(
     Gameplay=&AddVectorToPosition(
      Position=&SelectVectorFromArray(Direct Array Index=$Particles.PartitionParticles.Partition, Vector Selection Array=$System.Main Position)
      Vector=&SelectVectorFromArray(
       Direct Array Index=$Particles.PartitionParticles.Partition
       Vector Selection Array=DataInterface() {propertyValues {FloatData [<12 65 175>, <92 75 120>, <-20 32 190>]}}
      )
     )
     Front End=&AddVectorToPosition(
      Position=&SelectVectorFromArray(Direct Array Index=$Particles.PartitionParticles.Partition, Vector Selection Array=$System.Main Position)
      Vector=&AddVectorToPosition(
       Position=&SelectVectorFromArray(
        Direct Array Index=$Particles.PartitionParticles.Partition
        Vector Selection Array=DataInterface() {propertyValues {FloatData [<15 65 175>, <92 75 120>, <-20 32 190>]}}
       )
       Vector=&SelectVectorFromArray(Direct Array Index=$Particles.PartitionParticles.Partition, Vector Selection Array=$System.Front End Position)
      )
     )
    )
    UsePositionOffset=true
    Write Material Random=true
   )
   &CameraOffset(Camera Offset Amount=20)
   &InitialMeshOrientation(Mesh Orientation Mode=::NewEnumerator2, Rotation=<0 0.5 0>, Use Rotation Vector=true)
   &InitialMeshOrientation(
    Mesh Orientation Mode=::NewEnumerator2
    Rotation=&FN_FrontEndVector3Modifier(
     Gameplay=&SelectVectorFromArray(
      Direct Array Index=$Particles.PartitionParticles.Partition
      Vector Selection Array=DataInterface() {propertyValues {FloatData [<-0.05 0 0>, <0 0.08 0>, <0 0 0.5>]}}
     )
     Front End=&SelectVectorFromArray(
      Direct Array Index=$Particles.PartitionParticles.Partition
      Vector Selection Array=DataInterface() {propertyValues {FloatData [<-0.05 0 0>, <0 0 0>, <0 0 0.5>]}}
     )
    )
    Use Rotation Vector=true
   ) {module=InitialMeshOrientation001}
   &InitialMeshOrientation(
    Mesh Orientation Mode=::NewEnumerator2
    Rotation=&FN_FrontEndVector3Modifier(
     Gameplay=&SelectVectorFromArray(
      Direct Array Index=$Particles.PartitionParticles.Partition
      Vector Selection Array=DataInterface() {propertyValues {FloatData [<0 0 0>, <-0.1 0 0>, <-0.03 0 0>]}}
     )
     Front End=&SelectVectorFromArray(
      Direct Array Index=$Particles.PartitionParticles.Partition
      Vector Selection Array=DataInterface() {propertyValues {FloatData [<0 0 0>, <-0.03 0 0>, <-0.03 0 0>]}}
     )
    )
    Use Rotation Vector=true
   ) {module=InitialMeshOrientation002}
  ]
  particleUpdateScript [
   &ParticleState(Kill Particles When Lifetime Has Elapsed=false, Let Infinitely Lived Particles Die When Emitter Deactivates=true)
   &SkeletalMeshLocation(
    Bone Sampling Mode=::NewEnumerator4
    MaskSpawnGroup=false
    Orientation Sampling=::NewEnumerator1
    Sampled Position Offset=&Lerp_Vector(
     A=Linked() {linkedVariable {name=Particles.Initial.Position, type=NiagaraPosition}}
     B=&SelectVectorFromArray(
      Direct Array Index=$Particles.PartitionParticles.Partition
      Vector Selection Array=DataInterface() {propertyValues {FloatData [<0 0 200>, <0 0 -100>, <0 0 50>]}}
     )
     Alpha=&Waveform("Waveform [1]"=::NewEnumerator0, "[1] Frequency"=1, "[1] Amplitude Min/Max"=<0 1>, "[1] Scale"=0.1)
    )
    Sampled Position Offset Coordinate Space:ENiagaraSkelMeshTransforms=::NewEnumerator2
    Skeletal Mesh=$System.Skeletal Mesh Reader
   )
   &FN_Light_Attributes(
    Color=<rgba 15 8.5875 0.75 1>
    Create Color Attribute=true
    Create Exponent Attribute=true
    Create Position Attribute=true
    Create Radius Attribute=true
    Enable Particle Light Enable=true
    Exponent=2
    Light Enabled=true
    Position=&AddVectorToPosition(
     Position=$Particles.Position
     Vector=&SelectVectorFromArray(
      Direct Array Index=$Particles.PartitionParticles.Partition
      Vector Selection Array=DataInterface() {propertyValues {FloatData [<0 0 0>, <-150 0 0>, <150 0 0>]}}
     )
    )
    Radius=150
    Scale Color=&Multiply_Float(
     A=&FN_FrontEndFloatModifier(Gameplay=1, Front End=0.5)
     B=&Lerp_Float(A=$System.TestAlpha, B=1, Alpha=$User.ParaGlideLeanAlpha)
    )
   )
   &FN_Mesh_Attributes(
    Camera Offset=21
    Color=&MakeLinearColorFromVectorAndFloat(
     "Vector (RGB)"=&MakeVectorFromLinearColorRGB(Color=&Multiply_LinearColorByFloat(
       Linear Color=&MakeCustomLinearColorFromBool(Bool=$User.bIsFrontEndPreview, True Color=<rgba 80 9.3333 0 1>, False Color=<rgba 80 2.6667 0 1>)
       Float=&FN_FrontEndFloatModifier(
        Gameplay=0.5
        Front End=&FN_PickFloatDependOnPlatform(TargetPlatformActive=1, TargetPlatformNoActive=0.5)
       )
      ))
     "Float (Alpha)"=&Multiply_Float(
      A=&FN_FrontEndFloatModifier(Gameplay=1.5, Front End=1.5)
      B=&Lerp_Float(A=$System.TestAlpha, B=1, Alpha=&Subtract_Float(A=$User.ParaGlideLeanAlpha, B=$System.Threshold))
     )
    )
    Create Camera Offset Attribute=true
    Create Color Attribute=true
    Module Output=::NewEnumerator1
   )
  ]
  rendererClasses [&NiagaraMeshRendererProperties, &NiagaraLightRendererProperties]
  renderers [
   {rendererClass=&NiagaraMeshRendererProperties, rendererIndex=0}
   {rendererClass=&NiagaraLightRendererProperties, rendererIndex=1}
   {rendererClass=&NiagaraMeshRendererProperties, rendererIndex=2}
  ]
 }
 Sword001 {
  emitterUpdateScript [
   &EmitterState(Life Cycle Mode=::NewEnumerator1, Loop Duration=0.1, UseLoopDelay=false)
   &SpawnRate(
    Spawn Probability=&Lerp_Float(A=0.85, B=1, Alpha=$User.ParaGlideLeanAlpha)
    SpawnRate=30
    Use Spawn Probability=true
   )
  ]
  particleSpawnScript [
   &PartitionParticles(Number Of Partitions=3)
   &InitializeParticle(
    Color=&Multiply_LinearColorByFloat(Linear Color=<rgba 1 0.3667 0.05 10>, Float=&FN_FrontEndFloatModifier(Gameplay=0.8, Front End=1))
    Lifetime=0.4
    Mesh Scale=&Multiply_VectorByFloat(
     Vector=<1 1 1>
     Float=&FN_FrontEndFloatModifier(Gameplay=1.3, Front End=&Multiply_Float(A=1.3, B=$System.Front End Scale))
    )
    Mesh Scale Mode=::NewEnumerator4
    Position Offset=&FN_FrontEndVector3Modifier(
     Gameplay=&AddVectorToPosition(
      Position=&SelectVectorFromArray(Direct Array Index=$Particles.PartitionParticles.Partition, Vector Selection Array=$System.Main Position)
      Vector=&SelectVectorFromArray(
       Direct Array Index=$Particles.PartitionParticles.Partition
       Vector Selection Array=DataInterface() {propertyValues {FloatData [<12 65 175>, <92 75 120>, <-20 32 190>]}}
      )
     )
     Front End=&AddVectorToPosition(
      Position=&SelectVectorFromArray(Direct Array Index=$Particles.PartitionParticles.Partition, Vector Selection Array=$System.Main Position)
      Vector=&AddVectorToPosition(
       Position=&SelectVectorFromArray(
        Direct Array Index=$Particles.PartitionParticles.Partition
        Vector Selection Array=DataInterface() {propertyValues {FloatData [<15 65 175>, <92 75 120>, <-20 32 190>]}}
       )
       Vector=&SelectVectorFromArray(Direct Array Index=$Particles.PartitionParticles.Partition, Vector Selection Array=$System.Front End Position)
      )
     )
    )
    UsePositionOffset=true
    Write Material Random=true
   )
   &CameraOffset(Camera Offset Amount=-20)
   &InitialMeshOrientation(Mesh Orientation Mode=::NewEnumerator2, Rotation=<0 0.5 0>, Use Rotation Vector=true)
   &InitialMeshOrientation(
    Mesh Orientation Mode=::NewEnumerator2
    Rotation=&FN_FrontEndVector3Modifier(
     Gameplay=&SelectVectorFromArray(
      Direct Array Index=$Particles.PartitionParticles.Partition
      Vector Selection Array=DataInterface() {propertyValues {FloatData [<-0.05 0 0>, <0 0.08 0>, <0 0 0.5>]}}
     )
     Front End=&SelectVectorFromArray(
      Direct Array Index=$Particles.PartitionParticles.Partition
      Vector Selection Array=DataInterface() {propertyValues {FloatData [<-0.05 0 0>, <0 0 0>, <0 0 0.5>]}}
     )
    )
    Use Rotation Vector=true
   ) {module=InitialMeshOrientation001}
   &InitialMeshOrientation(
    Mesh Orientation Mode=::NewEnumerator2
    Rotation=&FN_FrontEndVector3Modifier(
     Gameplay=&SelectVectorFromArray(
      Direct Array Index=$Particles.PartitionParticles.Partition
      Vector Selection Array=DataInterface() {propertyValues {FloatData [<0 0 0>, <-0.1 0 0>, <-0.03 0 0>]}}
     )
     Front End=&SelectVectorFromArray(
      Direct Array Index=$Particles.PartitionParticles.Partition
      Vector Selection Array=DataInterface() {propertyValues {FloatData [<0 0 0>, <-0.03 0 0>, <-0.03 0 0>]}}
     )
    )
    Use Rotation Vector=true
   ) {module=InitialMeshOrientation002}
   &AddVelocity(
    Cone Angle=10
    Cone Axis=&FN_FrontEndVector3Modifier(
     Gameplay=<0 0 1>
     Front End=&SelectVectorFromArray(
      Direct Array Index=$Particles.PartitionParticles.Partition
      Vector Selection Array=DataInterface() {propertyValues {FloatData [<0 0.5 1>, <0 0.3 1>, <0 0.3 1>]}}
     )
    )
    Module.UseConeFalloff=false
    Module.UseVelDistribution=false
    Override Seed=false
    Velocity Mode=::NewEnumerator2
    Velocity Speed=&Lerp_Float(A=50, B=200, Alpha=$User.ParaGlideLeanAlpha)
   )
  ]
  particleUpdateScript [
   &ParticleState()
   &ScaleMeshSize(Scale Factor=&Multiply_VectorByFloat(
     Vector=<1 1 1>
     Float=&FloatFromCurve(FloatCurve=DataInterface() {propertyValues {Curve {keys [<at 0 1>,<at 1 0.8>]}}})
    ))
   &SkeletalMeshLocation(
    Bone Sampling Mode=::NewEnumerator4
    MaskSpawnGroup=false
    Orientation Sampling=::NewEnumerator1
    Position Sampling=::NewEnumerator1
    Sampled Position Offset=&Lerp_Vector(
     A=Linked() {linkedVariable {name=Particles.Initial.Position, type=NiagaraPosition}}
     B=&SelectVectorFromArray(
      Direct Array Index=$Particles.PartitionParticles.Partition
      Vector Selection Array=DataInterface() {propertyValues {FloatData [<0 0 200>, <0 0 -100>, <0 0 50>]}}
     )
     Alpha=&Waveform("Waveform [1]"=::NewEnumerator0, "[1] Frequency"=1, "[1] Amplitude Min/Max"=<0 1>, "[1] Scale"=0.1)
    )
    Sampled Position Offset Coordinate Space:ENiagaraSkelMeshTransforms=::NewEnumerator2
    Skeletal Mesh=$System.Skeletal Mesh Reader
    Source Movement Scale:Vector3f=<0.8 0.8 1>
   )
   &CurlNoiseForce(
    Bias Noise=false
    Noise Frequency=30
    Noise Strength=&Lerp_Float(A=0, B=200, Alpha=$User.ParaGlideLeanAlpha)
    Pan Noise=false
   )
   &SolveForcesAndVelocity(Clamp Velocity=false, Limit Acceleration=false)
   &ScaleColor(
    Scale Alpha=&FloatFromCurve(
     FloatCurve=DataInterface() {propertyValues {Curve {
        defaultValue=1
        keys [
         {interpMode=RCIM_Cubic, tangentMode=RCTM_User, time=0, value=0}
         {arriveTangent=-0.0543, interpMode=RCIM_Cubic, leaveTangent=-0.8958, tangentMode=RCTM_Break, time=0.3656, value=0.9418}
         {tangentMode=RCTM_User, time=1, value=0}
        ]
       }}}
     Scale Curve=&FloatFromCurve(
      FloatCurve=DataInterface() {propertyValues {Curve {keys [<at 0.3 0.8824>, {arriveTangent=-1.2605, interpMode=RCIM_Cubic, leaveTangent=-1.2605, time=0.8197, value=0.0853}, <at 1 0>]}}}
      Scale Curve=&FN_FrontEndFloatModifier(Gameplay=1.1, Front End=0.8)
     )
    )
    Scale RGB=&VectorFromFloat(Value=&FloatFromCurve(
      FloatCurve=DataInterface() {propertyValues {Curve {keys [<at 0 1>,<at 0.7 1.5>]}}}
      Scale Curve=&FN_FrontEndFloatModifier(Gameplay=0.8, Front End=0.8)
     ))
    ScaleA=true
    ScaleRGB=true
   )
  ]
  rendererClasses [&NiagaraMeshRendererProperties]
  renderers [{rendererClass=&NiagaraMeshRendererProperties, rendererIndex=0}]
 }
}
summary {
 systemName=NS_Contrail_LemoWrath02
 emitters [
  {emitterName=Sword, rendererClasses [&NiagaraMeshRendererProperties, &NiagaraLightRendererProperties]}
  {emitterName=Air, rendererClasses [&NiagaraMeshRendererProperties]}
  {emitterName=All_FoursTrail, rendererClasses [&NiagaraMeshRendererProperties]}
  {emitterName=MainTrail, rendererClasses [&NiagaraMeshRendererProperties]}
  {emitterName=MainTrail_L, rendererClasses [&NiagaraMeshRendererProperties]}
  {emitterName=MainTrail_R, rendererClasses [&NiagaraMeshRendererProperties]}
  {emitterName=EyesTrai, rendererClasses [&NiagaraMeshRendererProperties]}
  {emitterName=Sparks, rendererClasses [&NiagaraSpriteRendererProperties]}
  {emitterName=Eyes, rendererClasses [&NiagaraMeshRendererProperties]}
  {emitterName=Sword001, rendererClasses [&NiagaraMeshRendererProperties]}
 ]
 userVariables [
  User.bIsFrontEnd {type=NiagaraBool}
  User.bIsFrontEndPreview {type=NiagaraBool}
  User.ParaGlideLeanAlpha {type=NiagaraFloat, defaultValue=0, description=""}
 ]
}
